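# Material function for the Cactus Rocker backpack that samples recent audio spectrum and amplitude history at a 0-1 position.
# Custom HLSL nodes blend neighbors among four values from MPC_FilteredAudioParameterCollection, outputting interpolated spectrum and amplitude.
# Static switch useSmoothStep, default false, changes only amplitude from lerp to smoothstep with samples as edges, plus a small index-epsilon difference.
# The third and fourth spectrum inputs both read SpectrumMinus2.
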MaterialFunction /Game/Accessories/FORT_Backpacks/Backpack_F_MED_Cactus_Rocker/Materials/SampleAudioMPCInterpolateed
graph {nodes [
  MaterialExpressionCollectionParameter_0 = CollectionParameter() {collection=/Game/ArtTools/Tech_Art/AudioRT/MPC_FilteredAudioParameterCollection, parameterName=SpectrumNow}
  MaterialExpressionCollectionParameter_1 = CollectionParameter() {collection=&MPC_FilteredAudioParameterCollection, parameterName=SpectrumMinus1}
  MaterialExpressionCollectionParameter_2 = CollectionParameter() {collection=&MPC_FilteredAudioParameterCollection, parameterName=SpectrumMinus2}
  MaterialExpressionCollectionParameter_3 = CollectionParameter() {collection=&MPC_FilteredAudioParameterCollection, parameterName=SpectrumMinus2}
  MaterialExpressionFunctionInput_0 = FunctionInput(Preview=Constant()) {
   bUsePreviewValueAsDefault=true
   description=where in 0-1 space to sample the audio MPC and have it intepolated linearly
   inputName=uvx
   inputType=FunctionInput_Scalar
  }
  MaterialExpressionCollectionParameter_4 = CollectionParameter() {collection=&MPC_FilteredAudioParameterCollection, parameterName=AmplitudeSamplesPacked}
  n1 = Custom!(
   spectrum=@MaterialExpressionCollectionParameter_0
   spectrum_m1=@MaterialExpressionCollectionParameter_1
   spectrum_m2=@MaterialExpressionCollectionParameter_2
   spectrum_m3=@MaterialExpressionCollectionParameter_3
   alpha=@MaterialExpressionFunctionInput_0
   AmplitudePacked=@MaterialExpressionCollectionParameter_4
  ) {
   code="float4 values[4] = {\r\n\t\tspectrum, \r\n        spectrum_m1,\r\n\t\tspectrum_m2, \r\n        spectrum_m3\r\n\t};\r\n\r\n\r\nfloat extendRange = alpha*(4-1);\r\nfloat fractionalPart = frac(extendRange);\r\nuint index = uint(floor(extendRange));\r\n\r\nfloat4 mixedValue = lerp(values[index], values[index+1], fractionalPart);\r\n\r\nInterpolatedAmplitude =  smoothstep(AmplitudePacked[index], AmplitudePacked[index+1], fractionalPart);\r\nreturn mixedValue;"
   description=interpBetweenValuesSmoothStep
   outputType=CMOT_Float4
   additionalOutputs [{outputName=InterpolatedAmplitude, outputType=CMOT_Float1}]
   inputs [
    {
     inputName=spectrum
     input {expression=&self:MaterialExpressionCollectionParameter_0}
    }
    {
     inputName=spectrum_m1
     input {expression=&self:MaterialExpressionCollectionParameter_1}
    }
    {
     inputName=spectrum_m2
     input {expression=&self:MaterialExpressionCollectionParameter_2}
    }
    {
     inputName=spectrum_m3
     input {expression=&self:MaterialExpressionCollectionParameter_3}
    }
    {
     inputName=alpha
     input {expression=&self:MaterialExpressionFunctionInput_0}
    }
    {
     inputName=AmplitudePacked
     input {expression=&self:MaterialExpressionCollectionParameter_4}
    }
   ]
  }
  n2 = Custom(
   spectrum=@MaterialExpressionCollectionParameter_0
   spectrum_m1=@MaterialExpressionCollectionParameter_1
   spectrum_m2=@MaterialExpressionCollectionParameter_2
   spectrum_m3=@MaterialExpressionCollectionParameter_3
   alpha=@MaterialExpressionFunctionInput_0
   AmplitudePacked=@MaterialExpressionCollectionParameter_4
  ) {
   code="float4 values[4] = {\r\n\t\tspectrum, \r\n        spectrum_m1,\r\n\t\tspectrum_m2, \r\n        spectrum_m3\r\n\t};\r\n\r\n\r\nfloat extendRange = alpha*(4-1);\r\nfloat fractionalPart = frac(extendRange);\r\nuint index = uint(floor(extendRange+0.001));\r\n\r\nfloat4 mixedValue = lerp(values[index], values[index+1], fractionalPart);\r\n\r\nInterpolatedAmplitude =  lerp(AmplitudePacked[index], AmplitudePacked[index+1], fractionalPart);\r\nreturn mixedValue;"
   desc=Interpolates between historical samples of audio. Pass in a time value between 0-1 which will blend between 4 samples over a 32 sample history
   description=interpBetweenValues
   outputType=CMOT_Float4
   additionalOutputs [{outputName=InterpolatedAmplitude, outputType=CMOT_Float1}]
   inputs [
    {
     inputName=spectrum
     input {expression=&self:MaterialExpressionCollectionParameter_0}
    }
    {
     inputName=spectrum_m1
     input {expression=&self:MaterialExpressionCollectionParameter_1}
    }
    {
     inputName=spectrum_m2
     input {expression=&self:MaterialExpressionCollectionParameter_2}
    }
    {
     inputName=spectrum_m3
     input {expression=&self:MaterialExpressionCollectionParameter_3}
    }
    {
     inputName=alpha
     input {expression=&self:MaterialExpressionFunctionInput_0}
    }
    {
     inputName=AmplitudePacked
     input {expression=&self:MaterialExpressionCollectionParameter_4}
    }
   ]
  }
  n3 = FunctionInput(Preview=StaticBool()) {bUsePreviewValueAsDefault=true, desc="Default = False", inputName=useSmoothStep, inputType=FunctionInput_StaticBool}
  FunctionOutput(None=StaticSwitch(True=@n1.return, False=@n2.return, Value=@n3)) {outputName=Interpolated Spectrum}
  FunctionOutput(None=StaticSwitch(True=@n1.InterpolatedAmplitude, False=@n2.InterpolatedAmplitude, Value=@n3)) {outputName=Interpolated Amplitude Music}
 ]}
